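# Physics asset for the belt strings on the LastVoiceDive medium male player body cosmetic.
# Two capsule chains hang from a zero-size beltMain body, four links on the left and five on the right, each joint having limited swing and a zero twist limit.
# Pelvis and thigh capsules supply body volumes around the strings, with angularly free constraints linking beltMain and both thighs to the pelvis.
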PhysicsAsset /BRCosmetics/Characters/Player/Male/Medium/Bodies/M_MED_LastVoiceDive/Meshes/M_MED_LastVoiceDive_BeltStrings_Physics
bodies {
 beltMain [{shapeType=Capsule, length=0, radius=0, shapeName=beltMain_capsule}]
 beltString1_l [{shapeType=Capsule, length=4.5, radius=1, shapeName=beltString1_l_capsule, center=<4 0 0>, rotation=<rot 90 0 0>}]
 beltString1_r [{shapeType=Capsule, length=4.25, radius=1.25, shapeName=beltString1_r_capsule, center=<-4 0 0>, rotation=<rot 90 0 0>}]
 beltString2_l [{shapeType=Capsule, length=4.5, radius=1, shapeName=beltString2_l_capsule, center=<4 0 0>, rotation=<rot 90 0 0>}]
 beltString2_r [{shapeType=Capsule, length=4.25, radius=1.25, shapeName=beltString2_r_capsule, center=<-4 0 0>, rotation=<rot 90 0 0>}]
 beltString3_l [{shapeType=Capsule, length=4.5, radius=1, shapeName=beltString3_l_capsule, center=<4 0 0>, rotation=<rot 90 0 0>}]
 beltString3_r [{shapeType=Capsule, length=4.25, radius=1.25, shapeName=beltString3_r_capsule, center=<-4 0 0>, rotation=<rot 90 0 0>}]
 beltString4_l [{shapeType=Capsule, length=4.5, radius=1, shapeName=beltString4_l_capsule, center=<4 0 0>, rotation=<rot 90 0 0>}]
 beltString4_r [{shapeType=Capsule, length=4.25, radius=1.25, shapeName=beltString4_r_capsule, center=<-4 0 0>, rotation=<rot 90 0 0>}]
 beltString5_r [{shapeType=Capsule, length=4.25, radius=1.25, shapeName=beltString5_r_capsule, center=<-4 0 0>, rotation=<rot 90 0 0>}]
 pelvis [
  {shapeType=Capsule, length=3.5, radius=16.5, shapeName=pelvis_capsule, center=<0 -7.2908 0>}
  {shapeType=Capsule, length=14.25, radius=7, shapeName=pelvis_capsule, center=<-1.3529 3.2296 -0.1184>, rotation=<rot -85 -0.0003 -10>}
 ]
 thigh_l [{shapeType=Capsule, length=30.5, radius=10.25, shapeName=thigh_l_capsule, center=<-12.2973 0 0>, rotation=<rot -90 -90.3524 85.3522>}]
 thigh_r [{shapeType=Capsule, length=30.5, radius=10.25, shapeName=thigh_l_capsule, center=<12.2973 0 0>, rotation=<rot -90 -90.3524 85.3522>}]
}
constraints [
 {bone1Name=beltString1_l, bone2Name=beltMain, swing1LimitDegrees=45, swing2LimitDegrees=45, twistLimitDegrees=0}
 {bone1Name=beltString2_l, bone2Name=beltString1_l, swing1LimitDegrees=45, swing2LimitDegrees=45, twistLimitDegrees=0}
 {bone1Name=beltString3_l, bone2Name=beltString2_l, swing1LimitDegrees=45, swing2LimitDegrees=45, twistLimitDegrees=0}
 {bone1Name=beltString4_l, bone2Name=beltString3_l, swing1LimitDegrees=45, swing2LimitDegrees=45, twistLimitDegrees=0}
 {bone1Name=beltString1_r, bone2Name=beltMain, swing1LimitDegrees=45, swing2LimitDegrees=45, twistLimitDegrees=0}
 {bone1Name=beltString2_r, bone2Name=beltString1_r, swing1LimitDegrees=45, swing2LimitDegrees=45, twistLimitDegrees=0}
 {bone1Name=beltString3_r, bone2Name=beltString2_r, swing1LimitDegrees=45, swing2LimitDegrees=45, twistLimitDegrees=0}
 {bone1Name=beltString4_r, bone2Name=beltString3_r, swing1LimitDegrees=45, swing2LimitDegrees=45, twistLimitDegrees=0}
 {bone1Name=beltString5_r, bone2Name=beltString4_r, swing1LimitDegrees=45, swing2LimitDegrees=45, twistLimitDegrees=0}
 {
  bone1Name=beltMain
  bone2Name=pelvis
  swing1LimitDegrees=45
  swing1Motion=Free
  swing2LimitDegrees=45
  swing2Motion=Free
  twistLimitDegrees=45
  twistMotion=Free
 }
 {
  bone1Name=thigh_l
  bone2Name=pelvis
  swing1LimitDegrees=45
  swing1Motion=Free
  swing2LimitDegrees=45
  swing2Motion=Free
  twistLimitDegrees=45
  twistMotion=Free
 }
 {
  bone1Name=thigh_r
  bone2Name=pelvis
  swing1LimitDegrees=45
  swing1Motion=Free
  swing2LimitDegrees=45
  swing2Motion=Free
  twistLimitDegrees=45
  twistMotion=Free
 }
]
