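# Physics asset for the bottle string on the LastVoiceDive male medium body cosmetic.
# Two small capsules on the string bones and a larger capsule on the right upper-arm bottle bone make up the bodies.
# Constraints chain BottleString2 to BottleString1 and BottleString1 to upperarmBottle_r, each with narrow swing limit angles and a zero twist limit angle.
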PhysicsAsset /BRCosmetics/Characters/Player/Male/Medium/Bodies/M_MED_LastVoiceDive/Meshes/M_MED_LastVoiceDive_BottleString_Physics
bodies {
 BottleString1 [{shapeType=Capsule, length=3.5, radius=1, shapeName=BottleString1_capsule, center=<3 0 0>, rotation=<rot 90 0 0>}]
 BottleString2 [{shapeType=Capsule, length=3.5, radius=1, shapeName=BottleString2_capsule, center=<3 0 0>, rotation=<rot 90 0 0>}]
 upperarmBottle_r [{
   shapeType=Capsule
   length=24.2931
   radius=7.3512
   shapeName=upperarmBottle_r_capsule
   center=<-1.712 0.8134 0.8824>
   rotation=<rot -0.1962 -79.8176 -88.8882>
  }]
}
constraints [
 {bone1Name=BottleString1, bone2Name=upperarmBottle_r, swing1LimitDegrees=10, swing2LimitDegrees=10, twistLimitDegrees=0}
 {bone1Name=BottleString2, bone2Name=BottleString1, swing1LimitDegrees=10, swing2LimitDegrees=10, twistLimitDegrees=0}
]
