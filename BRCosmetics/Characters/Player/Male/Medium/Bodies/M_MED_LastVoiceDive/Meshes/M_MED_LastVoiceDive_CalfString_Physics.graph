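# Physics asset for the right-calf string on the LastVoiceDive medium male player body cosmetic, letting the string be simulated as a short bone chain.
# A large calf_r capsule anchors a zero-size calfStringMaster body, followed by two thin string capsules.
# Three constraints link the chain with swing and twist all Free, so their 45-degree limits have no effect.
# The string capsules carry beltString left-side shape names, unlike their bones.
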PhysicsAsset /BRCosmetics/Characters/Player/Male/Medium/Bodies/M_MED_LastVoiceDive/Meshes/M_MED_LastVoiceDive_CalfString_Physics
bodies {
 calfString1_r [{shapeType=Capsule, length=4.5, radius=1, shapeName=beltString1_l_capsule, center=<4 0 0>, rotation=<rot 90 0 0>}]
 calfString2_r [{shapeType=Capsule, length=4.5, radius=1, shapeName=beltString2_l_capsule, center=<4 0 0>, rotation=<rot 90 0 0>}]
 calfStringMaster [{shapeType=Capsule, length=0, radius=0, shapeName=calfStringMaster_capsule, rotation=<rot 0 -90 0>}]
 calf_r [{shapeType=Capsule, length=11.4038, radius=7.8353, shapeName=calf_r_capsule, center=<26.8232 0 2.0704>, rotation=<rot 90 179.9802 179.9803>}]
}
constraints [
 {
  bone1Name=calfString1_r
  bone2Name=calfStringMaster
  swing1LimitDegrees=45
  swing1Motion=Free
  swing2LimitDegrees=45
  swing2Motion=Free
  twistLimitDegrees=45
  twistMotion=Free
 }
 {
  bone1Name=calfString2_r
  bone2Name=calfString1_r
  swing1LimitDegrees=45
  swing1Motion=Free
  swing2LimitDegrees=45
  swing2Motion=Free
  twistLimitDegrees=45
  twistMotion=Free
 }
 {
  bone1Name=calfStringMaster
  bone2Name=calf_r
  swing1LimitDegrees=45
  swing1Motion=Free
  swing2LimitDegrees=45
  swing2Motion=Free
  twistLimitDegrees=45
  twistMotion=Free
 }
]
