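# Physics asset for the back tag mesh of the male medium LastVoiceDive player body, covering a hanging tag and two tennis balls.
# Pelvis and tagMaster are zero-size capsule anchors, while the tag and balls share an identical small capsule shape.
# The tag chains through tagMaster, which rotates freely on the pelvis.
# Both balls attach directly to the pelvis, and they and the tag have narrow swing limits with zero twist range.
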PhysicsAsset /Game/Characters/Player/Male/Medium/Bodies/M_MED_LastVoiceDive/Meshes/M_MED_LastVoiceDive_BackTag_Physics
bodies {
 pelvis [{shapeType=Capsule, length=0, radius=0, shapeName=pelvis_capsule}]
 tag [{shapeType=Capsule, length=4.5, radius=1, shapeName=beltString1_l_capsule, center=<4 0 0>, rotation=<rot 90 0 0>}]
 tagMaster [{shapeType=Capsule, length=0, radius=0, shapeName=tagMaster_capsule}]
 tennisBall1 [{shapeType=Capsule, length=4.5, radius=1, shapeName=beltString1_l_capsule, center=<4 0 0>, rotation=<rot 90 0 0>}]
 tennisBall2 [{shapeType=Capsule, length=4.5, radius=1, shapeName=beltString1_l_capsule, center=<4 0 0>, rotation=<rot 90 0 0>}]
}
constraints [
 {bone1Name=tag, bone2Name=tagMaster, swing1LimitDegrees=30, swing2LimitDegrees=30, twistLimitDegrees=0}
 {
  bone1Name=tagMaster
  bone2Name=pelvis
  swing1LimitDegrees=45
  swing1Motion=Free
  swing2LimitDegrees=45
  swing2Motion=Free
  twistLimitDegrees=45
  twistMotion=Free
 }
 {bone1Name=tennisBall1, bone2Name=pelvis, swing1LimitDegrees=20, swing2LimitDegrees=20, twistLimitDegrees=0}
 {bone1Name=tennisBall2, bone2Name=pelvis, swing1LimitDegrees=20, swing2LimitDegrees=20, twistLimitDegrees=0}
]
