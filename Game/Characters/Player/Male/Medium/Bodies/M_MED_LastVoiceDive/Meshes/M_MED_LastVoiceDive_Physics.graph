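# Physics asset for the LastVoiceDive medium male player body, covering only the fluff on the left elbow.
# A small capsule body on elbowFluff_l is constrained to a zero-size capsule on upperarm_l, which serves as its anchor.
# The single joint allows a narrow swing cone and no twist, which suggests limited secondary motion for the fluff.
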PhysicsAsset /Game/Characters/Player/Male/Medium/Bodies/M_MED_LastVoiceDive/Meshes/M_MED_LastVoiceDive_Physics
bodies {
 elbowFluff_l [{shapeType=Capsule, length=1, radius=3, shapeName=elbowFluff_l_capsule, center=<5 0 0>, rotation=<rot 90 0 0>}]
 upperarm_l [{shapeType=Capsule, length=0, radius=0, shapeName=upperarm_l_capsule}]
}
constraints [{bone1Name=elbowFluff_l, bone2Name=upperarm_l, swing1LimitDegrees=15, swing2LimitDegrees=15, twistLimitDegrees=0}]
