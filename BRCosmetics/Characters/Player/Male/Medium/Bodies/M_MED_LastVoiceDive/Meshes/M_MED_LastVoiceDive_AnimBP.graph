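# Post-process style cosmetic AnimBP for the LastVoiceDive player body, copying the attached parent's pose and adding secondary motion.
# Control Rigs and weighted constraints position deform bones and accessories, then RigidBody and AnimDynamics simulate strings, tag, tennis balls and elbow fluff.
# Additive bone offsets correct collar, arm, bottle and belt string for skydiving, hang gliding, rifle and idle states.
# EventGraph derives those alphas, gravity override and a Mime-emote check enabling separate belt-string physics.
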AnimBlueprint /BRCosmetics/Characters/Player/Male/Medium/Bodies/M_MED_LastVoiceDive/Meshes/M_MED_LastVoiceDive_AnimBP
target_skeleton=&/M_MED_LastVoiceDive_Skeleton
graphs [
 AnimGraph {nodes [
   RigidBody!(ComponentPose, Alpha=0.75, OverrideWorldGravity:Vector="0,0,0", ExternalForce) -> (Pose) {
    bClampLinearTranslationLimitToRefPose=true
    bOverrideWorldGravity=true
    overridePhysicsAsset=/Game/Characters/Player/Male/Medium/Bodies/M_MED_LastVoiceDive/Meshes/M_MED_LastVoiceDive_Physics
    simulationSpace=BaseBoneSpace
    alphaBoolBlend {blendOption=Linear}
    baseBoneRef {boneName=spine_05}
    componentAppliedLinearAccClamp=<3000 3000 3000>
    componentLinearAccScale=<0.4 0.3 0.3>
    componentLinearVelScale=<0.4 0.3 0.3>
    overrideWorldGravity {z=-980}
    simSpaceSettings {
     maxAngularAcceleration=1000
     maxAngularVelocity=3
     maxLinearAcceleration=250
     maxLinearVelocity=500
     velocityScaleZ=0
     worldAlpha=0.25
     externalLinearDragV {y=0.5}
    }
   }
   RigidBody(OverrideWorldGravity="0,0,0") {
    bClampLinearTranslationLimitToRefPose=true
    bOverrideWorldGravity=true
    overridePhysicsAsset=/Game/Characters/Player/Male/Medium/Bodies/M_MED_LastVoiceDive/Meshes/M_MED_LastVoiceDive_BackTag_Physics
    simulationSpace=BaseBoneSpace
    alphaBoolBlend {blendOption=Linear}
    baseBoneRef {boneName=spine_01}
    componentAppliedLinearAccClamp=<3000 3000 3000>
    componentLinearAccScale=<0.4 0.3 0.3>
    componentLinearVelScale=<0.4 0.3 0.3>
    overrideWorldGravity {z=-980}
    simSpaceSettings {
     maxAngularAcceleration=1000
     maxAngularVelocity=3
     maxLinearAcceleration=250
     maxLinearVelocity=500
     velocityScaleZ=0
     worldAlpha=0.25
     externalLinearDragV {y=0.5}
    }
   }
   n1 = Control Rig(Source=Copy Pose From Mesh() {bUseAttachedParent=true}) {
    alphaBoolBlend {blendOption=Linear}
    controlRigAssetReference {blueprintRigClass=&/M_MED_DeformRig_LastVoiceDive_ControlRig_new1.M_MED_DeformRig_LastVoiceDive_ControlRig_new1_C}
   }
   n2 = Control Rig!(Source=@n1, bAlphaBoolEnabled:bool=NOT Boolean(A=Get bIsSkyDiving())) -> (Pose) {
    alphaInputType=Bool
    alphaBoolBlend {blendOption=Linear}
    controlRigAssetReference {blueprintRigClass=&/M_MED_LastVoiceDive_Neck_CtrlRig.M_MED_LastVoiceDive_Neck_CtrlRig_C}
   }
   n3 = Control Rig!(Source=@n2, bAlphaBoolEnabled=True, isCrouchingNotMoving:double=Get isCrouchMoving() -> (isCrouchMoving:double)) -> (Pose) {
    alphaInputType=Bool
    alphaBoolBlend {blendOption=Linear}
    controlRigAssetReference {blueprintRigClass=&/M_MED_LastVoiceDive_NeckCrouch_CtrlRig.M_MED_LastVoiceDive_NeckCrouch_CtrlRig_C}
   }
   n4 = Control Rig!(Source=@n3, Alpha:float, isCrouching:double=0) -> (Pose) {
    alphaBoolBlend {blendOption=Linear}
    controlRigAssetReference {blueprintRigClass=&/M_MED_LastVoiceDive_Pistol_CtrlRig.M_MED_LastVoiceDive_Pistol_CtrlRig_C}
   }
   n5 = Control Rig!(Source=@n4, Alpha) -> (Pose) {
    alphaBoolBlend {blendOption=Linear}
    controlRigAssetReference {blueprintRigClass=&/M_MED_LastVoiceDive_armBottle_CtrlRig.M_MED_LastVoiceDive_armBottle_CtrlRig_C}
   }
   n6 = "Constraint - pouch (PPPPPPPPPPPP)"!(
    ComponentPose=Local To Component(LocalPose=@n5)
    Alpha=1
    ConstraintWeights_0=0.6905
    ConstraintWeights_1:float=0.0025
    ConstraintWeights_2:float=0.0029
    ConstraintWeights_3:float=0.0013
    ConstraintWeights_4:float=0
    ConstraintWeights_5:float=0.2158
    ConstraintWeights_6:float=0.0126
    ConstraintWeights_7:float=0.0744
    ConstraintWeights_8:float=0
    ConstraintWeights_9:float=0
    ConstraintWeights_10:float=0
    ConstraintWeights_11:float=0.0001
   ) -> (Pose) {
    alphaBoolBlend {blendOption=Linear}
    boneToModify {boneName=pouch}
    constraintSetup [
     {targetBone {boneName=pelvis}}
     {targetBone {boneName=spine_01}}
     {targetBone {boneName=spine_02}}
     {targetBone {boneName=spine_03}}
     {targetBone {boneName=thigh_l}}
     {targetBone {boneName=thigh_twist_01_l}}
     {targetBone {boneName=deform_glute_l}}
     {targetBone {boneName=deform_groin_l}}
     {targetBone {boneName=thigh_r}}
     {targetBone {boneName=thigh_twist_01_r}}
     {targetBone {boneName=deform_glute_r}}
     {targetBone {boneName=deform_groin_r}}
    ]
    constraintWeights [1,1,1,1,1,1,1,1,1,1,1,1]
   }
   n7 = "Constraint - pump (PPPP)"!(
    ComponentPose=@n6
    Alpha=1
    ConstraintWeights_0=0.0524
    ConstraintWeights_1:float=0.1308
    ConstraintWeights_2:float=0.6941
    ConstraintWeights_3:float=0.1227
   ) -> (Pose) {
    alphaBoolBlend {blendOption=Linear}
    boneToModify {boneName=pump}
    constraintSetup [
     {targetBone {boneName=pelvis}}
     {targetBone {boneName=thigh_l}}
     {targetBone {boneName=thigh_twist_01_l}}
     {targetBone {boneName=deform_groin_l}}
    ]
    constraintWeights [1,1,1,1]
   }
   n8 = Control Rig(Source=Component To Local(ComponentPose=@n7)) {
    alphaBoolBlend {blendOption=Linear}
    controlRigAssetReference {blueprintRigClass=&/M_MED_LastVoiceDive_Pouch_CtrlRig.M_MED_LastVoiceDive_Pouch_CtrlRig_C}
   }
   n9 = "Constraint - tennisBall1 (PPPPP)"!(
    ComponentPose=Local To Component(LocalPose=@n8)
    Alpha=1
    ConstraintWeights_0=0.6538
    ConstraintWeights_1:float=0.1212
    ConstraintWeights_2:float=0.1424
    ConstraintWeights_3:float=0.0823
    ConstraintWeights_4:float=0.0003
   ) -> (Pose) {
    alphaBoolBlend {blendOption=Linear}
    boneToModify {boneName=tennisBall1}
    constraintSetup [
     {targetBone {boneName=pelvis}}
     {targetBone {boneName=spine_01}}
     {targetBone {boneName=spine_02}}
     {targetBone {boneName=spine_03}}
     {targetBone {boneName=spine_04}}
    ]
    constraintWeights [1,1,1,1,1]
   }
   n10 = "Constraint - tennisBall2 (PPPPPP)"!(
    ComponentPose=@n9
    Alpha=1
    ConstraintWeights_0=0.8156
    ConstraintWeights_1:float=0.0539
    ConstraintWeights_2:float=0.0628
    ConstraintWeights_3:float=0.0345
    ConstraintWeights_4:float=0.023
    ConstraintWeights_5:float=0.0102
   ) -> (Pose) {
    alphaBoolBlend {blendOption=Linear}
    boneToModify {boneName=tennisBall2}
    constraintSetup [
     {targetBone {boneName=pelvis}}
     {targetBone {boneName=spine_01}}
     {targetBone {boneName=spine_02}}
     {targetBone {boneName=spine_03}}
     {targetBone {boneName=thigh_twist_01_r}}
     {targetBone {boneName=deform_groin_r}}
    ]
    constraintWeights [1,1,1,1,1,1]
   }
   n11 = "Constraint - beltMain (PPPPPPPP)"!(
    ComponentPose=@n10
    Alpha=1
    ConstraintWeights_0=0.6395
    ConstraintWeights_1:float=0.1219
    ConstraintWeights_2:float=0.1447
    ConstraintWeights_3:float=0.087
    ConstraintWeights_4:float=0.0057
    ConstraintWeights_5:float=0.0004
    ConstraintWeights_6:float=0.0004
    ConstraintWeights_7:float=0.0004
   ) -> (Pose) {
    alphaBoolBlend {blendOption=Linear}
    boneToModify {boneName=beltMain}
    constraintSetup [
     {targetBone {boneName=pelvis}}
     {targetBone {boneName=spine_01}}
     {targetBone {boneName=spine_02}}
     {targetBone {boneName=spine_03}}
     {targetBone {boneName=spine_04}}
     {targetBone {boneName=thigh_twist_01_l}}
     {targetBone {boneName=deform_groin_l}}
     {targetBone {boneName=deform_groin_r}}
    ]
    constraintWeights [1,1,1,1,1,1,1,1]
   }
   n12 = "Constraint - calfStringMaster (PPPPPPPP)"!(
    ComponentPose=@n11
    Alpha=1
    ConstraintWeights_0=0.0068
    ConstraintWeights_1:float=0.3237
    ConstraintWeights_2:float=0.0518
    ConstraintWeights_3:float=0.6134
    ConstraintWeights_4:float=0
    ConstraintWeights_5:float=0.0038
    ConstraintWeights_6:float=0.0005
    ConstraintWeights_7:float=0
   ) -> (Pose) {
    alphaBoolBlend {blendOption=Linear}
    boneToModify {boneName=calfStringMaster}
    constraintSetup [
     {targetBone {boneName=thigh_r}}
     {targetBone {boneName=calf_r}}
     {targetBone {boneName=calf_twist_01_r}}
     {targetBone {boneName=calf_twist_02_r}}
     {targetBone {boneName=foot_r}}
     {targetBone {boneName=deform_knee_out_r}}
     {targetBone {boneName=deform_knee_in_r}}
     {targetBone {boneName=thigh_twist_01_r}}
    ]
    constraintWeights [1,1,1,1,1,1,1,1]
   }
   n13 = "Constraint - elbowFluff_l (PPPPPPPP)"!(
    ComponentPose=@n12
    Alpha=1
    ConstraintWeights_0=0.3498
    ConstraintWeights_1:float=0.2397
    ConstraintWeights_2:float=0.007
    ConstraintWeights_3:float=0.1561
    ConstraintWeights_4:float=0.1388
    ConstraintWeights_5:float=0.0583
    ConstraintWeights_6:float=0.001
    ConstraintWeights_7:float=0.0494
   ) -> (Pose) {
    alphaBoolBlend {blendOption=Linear}
    boneToModify {boneName=elbowFluff_l}
    constraintSetup [
     {targetBone {boneName=upperarm_l}}
     {targetBone {boneName=lowerarm_l}}
     {targetBone {boneName=lowerarm_twist_01_l}}
     {targetBone {boneName=lowerarm_twist_02_l}}
     {targetBone {boneName=deform_elbow_out_l}}
     {targetBone {boneName=deform_elbow_in_l}}
     {targetBone {boneName=upperarm_twist_01_l}}
     {targetBone {boneName=upperarm_twist_02_l}}
    ]
    constraintWeights [1,1,1,1,1,1,1,1]
   }
   n14 = Get GravityOverride()
   n15 = RigidBody(ComponentPose=@n13, OverrideWorldGravity=@n14) {
    bClampLinearTranslationLimitToRefPose=true
    bOverrideWorldGravity=true
    overridePhysicsAsset=&/M_MED_LastVoiceDive_BeltStrings_Physics
    simulationSpace=BaseBoneSpace
    alphaBoolBlend {blendOption=Linear}
    baseBoneRef {boneName=pelvis}
    componentAppliedLinearAccClamp=<3000 3000 3000>
    componentLinearAccScale=<0.4 0.3 0.3>
    componentLinearVelScale=<0.4 0.3 0.3>
    overrideWorldGravity {z=-980}
    simSpaceSettings {
     maxAngularAcceleration=1000
     maxAngularVelocity=3
     maxLinearAcceleration=250
     maxLinearVelocity=500
     velocityScaleZ=0
     worldAlpha=0.25
     externalLinearDragV {y=0.5}
    }
   }
   n16 = RigidBody(ComponentPose=@n15, Alpha=Get IsCustomEmotePlaying() -> (IsCustomEmotePlaying:double), OverrideWorldGravity=@n14) {
    bClampLinearTranslationLimitToRefPose=true
    bOverrideWorldGravity=true
    overridePhysicsAsset=&/M_MED_LastVoiceDive_BeltStringsEmote_Physics
    simulationSpace=BaseBoneSpace
    alphaBoolBlend {blendOption=Linear}
    baseBoneRef {boneName=pelvis}
    componentAppliedLinearAccClamp=<3000 3000 3000>
    componentLinearAccScale=<0.4 0.3 0.3>
    componentLinearVelScale=<0.4 0.3 0.3>
    overrideWorldGravity {z=-980}
    simSpaceSettings {
     maxAngularAcceleration=1000
     maxAngularVelocity=3
     maxLinearAcceleration=250
     maxLinearVelocity=500
     velocityScaleZ=0
     worldAlpha=0.25
     externalLinearDragV {y=0.5}
    }
   }
   n17 = RigidBody(ComponentPose=@n16, OverrideWorldGravity=@n14) {
    bClampLinearTranslationLimitToRefPose=true
    bOverrideWorldGravity=true
    overridePhysicsAsset=&/M_MED_LastVoiceDive_CalfString_Physics
    simulationSpace=BaseBoneSpace
    alphaBoolBlend {blendOption=Linear}
    baseBoneRef {boneName=thigh_r}
    componentAppliedLinearAccClamp=<3000 3000 3000>
    componentLinearAccScale=<0.4 0.3 0.3>
    componentLinearVelScale=<0.4 0.3 0.3>
    overrideWorldGravity {z=-980}
    simSpaceSettings {
     maxAngularAcceleration=1000
     maxAngularVelocity=3
     maxLinearAcceleration=250
     maxLinearVelocity=500
     velocityScaleZ=0
     worldAlpha=0.25
     externalLinearDragV {y=0.5}
    }
   }
   n18 = RigidBody(ComponentPose=@n17, OverrideWorldGravity=@n14) {
    bClampLinearTranslationLimitToRefPose=true
    bOverrideWorldGravity=true
    overridePhysicsAsset=&/M_MED_LastVoiceDive_BottleString_Physics
    simulationSpace=BaseBoneSpace
    alphaBoolBlend {blendOption=Linear}
    baseBoneRef {boneName=upperarm_r}
    componentAppliedLinearAccClamp=<3000 3000 3000>
    componentLinearAccScale=<0.4 0.3 0.3>
    componentLinearVelScale=<0.4 0.3 0.3>
    overrideWorldGravity {z=-980}
    simSpaceSettings {
     maxAngularAcceleration=1000
     maxAngularVelocity=3
     maxLinearAcceleration=250
     maxLinearVelocity=500
     velocityScaleZ=0
     worldAlpha=0.25
     externalLinearDragV {y=0.5}
    }
   }
   n19 = Anim Dynamics - Bone: tagMaster(ComponentPose=@n18, Alpha=1, ExternalForce="0,0,0") {
    angularDampingOverride=1
    angularSpringConstant=5000
    bAngularSpring=true
    bOverrideAngularDamping=true
    bUseGravityOverride=true
    gravityScale=0.35
    simulationSpace=RootRelative
    alphaBoolBlend {blendOption=Linear}
    boundBone {boneName=tagMaster}
    componentLinearAccScale=<1 1 1>
    componentLinearVelScale=<1 1 1>
    gravityOverride {z=-1500}
    physicsBodyDefinitions [{
      boundBone {boneName=tagMaster}
      boxExtents=<18 18 18>
      constraintSetup {angularLimitsMax=<30 30 30>, angularLimitsMin=<-30 -30 -30>, angularTarget=<1 0 0>}
      localJointOffset=<1 0 0>
     }]
   }
   n20 = "Transform (Modify) Bone - Bone: tag"!(ComponentPose=@n19, bAlphaBoolEnabled:bool=True, Translation="0,0,0", Rotation="30,0,0", Scale="1,1,1") -> (Pose) {
    alphaInputType=Bool
    rotationMode=BMM_Additive
    rotationSpace=BCS_BoneSpace
    translationMode=BMM_Additive
    translationSpace=BCS_BoneSpace
    alphaBoolBlend {blendOption=Linear}
    boneToModify {boneName=tag}
    rotation {pitch=30}
   }
   n21 = Anim Dynamics - Bone: tennisBall1(ComponentPose=@n20, Alpha=0.5, ExternalForce="0,0,0") {
    angularDampingOverride=1
    angularSpringConstant=5000
    bAngularSpring=true
    bOverrideAngularDamping=true
    bUseGravityOverride=true
    gravityScale=0.35
    simulationSpace=RootRelative
    alphaBoolBlend {blendOption=Linear}
    boundBone {boneName=tennisBall1}
    componentLinearAccScale=<1 1 1>
    componentLinearVelScale=<1 1 1>
    gravityOverride {z=-1500}
    physicsBodyDefinitions [{
      boundBone {boneName=tennisBall1}
      boxExtents=<12 12 12>
      constraintSetup {angularLimitsMax=<0 30 0>, angularLimitsMin=<0 -30 0>, angularTarget=<1 0 0>}
      localJointOffset=<1 0 0>
     }]
   }
   n22 = Anim Dynamics - Bone: tennisBall2(ComponentPose=@n21, Alpha=0.5, ExternalForce="0,0,0") {
    angularDampingOverride=1
    angularSpringConstant=5000
    bAngularSpring=true
    bOverrideAngularDamping=true
    bUseGravityOverride=true
    gravityScale=0.35
    simulationSpace=RootRelative
    alphaBoolBlend {blendOption=Linear}
    boundBone {boneName=tennisBall2}
    componentLinearAccScale=<1 1 1>
    componentLinearVelScale=<1 1 1>
    gravityOverride {z=-1500}
    physicsBodyDefinitions [{
      boundBone {boneName=tennisBall2}
      boxExtents=<15 15 15>
      constraintSetup {angularLimitsMax=<0 30 0>, angularLimitsMin=<0 -30 0>, angularTarget=<1 0 0>}
      localJointOffset=<1 0 0>
     }]
   }
   n23 = Anim Dynamics - Bone: elbowFluff_l(ComponentPose=@n22, Alpha=1, ExternalForce="0,0,0") {
    angularDampingOverride=1
    angularSpringConstant=5000
    bAngularSpring=true
    bOverrideAngularDamping=true
    bUseGravityOverride=true
    gravityScale=0.35
    simulationSpace=RootRelative
    alphaBoolBlend {blendOption=Linear}
    boundBone {boneName=elbowFluff_l}
    componentLinearAccScale=<1 1 1>
    componentLinearVelScale=<1 1 1>
    gravityOverride {z=-1500}
    physicsBodyDefinitions [{
      boundBone {boneName=elbowFluff_l}
      boxExtents=<9 9 9>
      constraintSetup {angularLimitsMax=<5 30 5>, angularLimitsMin=<-5 -30 -5>, angularTarget=<1 0 0>}
      localJointOffset=<1 0 0>
     }]
   }
   n24 = "Constraint - upperarm_twist_02_l (R)"(ComponentPose=@n23, Alpha=1, ConstraintWeights_0=1) {
    alphaBoolBlend {blendOption=Linear}
    boneToModify {boneName=upperarm_twist_02_l}
    constraintSetup [{transformType=Rotation, targetBone {boneName=upperarm_l}}]
    constraintWeights [1]
   }
   n25 = "Constraint - upperarmBottle_r (R)"(ComponentPose=@n24, Alpha=1, ConstraintWeights_0=1) {
    alphaBoolBlend {blendOption=Linear}
    boneToModify {boneName=upperarmBottle_r}
    constraintSetup [{transformType=Rotation, targetBone {boneName=upperarm_twist_02_r}}]
    constraintWeights [1]
   }
   n26 = Get isSkydiveFloatingFloat() -> (isSkydiveFloatingFloat:float)
   n27 = "Transform (Modify) Bone - Bone: upperarmBottle_r"(ComponentPose=@n25, Alpha=@n26, Translation="-3,0,-3", Rotation="0,0,0", Scale="1,1,1") {
    rotationMode=BMM_Additive
    rotationSpace=BCS_BoneSpace
    translationMode=BMM_Additive
    translationSpace=BCS_BoneSpace
    alphaBoolBlend {blendOption=Linear}
    alphaScaleBiasClamp {bInterpResult=true}
    boneToModify {boneName=upperarmBottle_r}
    translation {x=-3, z=-3}
   }
   n28 = "Transform (Modify) Bone - Bone: upperarm_twist_02_l"(ComponentPose=@n27, Alpha=@n26, Translation="0,-5,0", Rotation="20,0,0", Scale="1,1,1") {
    rotationMode=BMM_Additive
    rotationSpace=BCS_BoneSpace
    translationMode=BMM_Additive
    translationSpace=BCS_BoneSpace
    alphaBoolBlend {blendOption=Linear}
    alphaScaleBiasClamp {bInterpResult=true}
    boneToModify {boneName=upperarm_twist_02_l}
    rotation {pitch=20}
    translation {y=-5}
   }
   n29 = "Transform (Modify) Bone - Bone: upperarmBottle_r"(ComponentPose=@n28, Alpha=Get IsHangGlider() -> (IsHangGlider:double), Translation="-7,0,-3", Rotation="-10,0,0", Scale="1,1,1") {
    rotationMode=BMM_Additive
    rotationSpace=BCS_BoneSpace
    translationMode=BMM_Additive
    translationSpace=BCS_BoneSpace
    alphaBoolBlend {blendOption=Linear}
    boneToModify {boneName=upperarmBottle_r}
    rotation {pitch=-10}
    translation {x=-7, z=-3}
   }
   n30 = Get bIsSkyDiving()
   n31 = "Constraint - collar_01 (P)"!(ComponentPose=@n29, bAlphaBoolEnabled:bool=@n30, ConstraintWeights_0=1) -> (Pose) {
    alphaInputType=Bool
    alphaBoolBlend {blendOption=Linear}
    boneToModify {boneName=collar_01}
    constraintSetup [{targetBone {boneName=spine_05}}]
    constraintWeights [1]
   }
   n32 = "Constraint - collar_l__01 (P)"!(ComponentPose=@n31, bAlphaBoolEnabled:bool=True, ConstraintWeights_0=1) -> (Pose) {
    alphaInputType=Bool
    alphaBoolBlend {blendOption=Linear}
    boneToModify {boneName=collar_l__01}
    constraintSetup [{targetBone {boneName=collar_01}}]
    constraintWeights [1]
   }
   n33 = "Constraint - collar_r__01 (P)"!(ComponentPose=@n32, bAlphaBoolEnabled:bool=True, ConstraintWeights_0=1) -> (Pose) {
    alphaInputType=Bool
    alphaBoolBlend {blendOption=Linear}
    boneToModify {boneName=collar_r__01}
    constraintSetup [{targetBone {boneName=collar_01}}]
    constraintWeights [1]
   }
   n34 = "Transform (Modify) Bone - Bone: collar_01"!(ComponentPose=@n33, bAlphaBoolEnabled:bool=@n30, Translation="-3,-5,0", Rotation="0,0,0", Scale="1,1,1") -> (Pose) {
    alphaInputType=Bool
    rotationMode=BMM_Additive
    rotationSpace=BCS_BoneSpace
    translationMode=BMM_Additive
    translationSpace=BCS_BoneSpace
    alphaBoolBlend {blendOption=Linear}
    boneToModify {boneName=collar_01}
    translation=<-3 -5>
   }
   n35 = "Transform (Modify) Bone - Bone: upperarmBottle_r"(ComponentPose=@n34, Alpha=Get rifleNotTargeting() -> (rifleNotTargeting:float), Translation="-2,0,-1", Rotation="-10,0,0", Scale="1,1,1") {
    rotationMode=BMM_Additive
    rotationSpace=BCS_BoneSpace
    translationMode=BMM_Additive
    translationSpace=BCS_BoneSpace
    alphaBoolBlend {blendOption=Linear}
    alphaScaleBiasClamp {bInterpResult=true, interpSpeedDecreasing=18}
    boneToModify {boneName=upperarmBottle_r}
    rotation {pitch=-10}
    translation {x=-2, z=-1}
   }
   n36 = Get IsHangGlider()
   n37 = "Transform (Modify) Bone - Bone: collar_01"!(ComponentPose=@n35, Alpha:float=@n36, Translation="3,0,0", Rotation="0,0,0", Scale="1,1,1") -> (Pose) {
    rotationMode=BMM_Additive
    rotationSpace=BCS_BoneSpace
    translationMode=BMM_Additive
    translationSpace=BCS_BoneSpace
    alphaBoolBlend {blendOption=Linear}
    alphaScaleBiasClamp {bInterpResult=true}
    boneToModify {boneName=collar_01}
    translation {x=3}
   }
   n38 = "Transform (Modify) Bone - Bone: collar_l__01"(ComponentPose=@n37, Alpha=@n36, Translation="0,0,-1", Rotation="0,0,0", Scale="1,1,1") {
    rotationMode=BMM_Additive
    rotationSpace=BCS_BoneSpace
    translationMode=BMM_Additive
    translationSpace=BCS_BoneSpace
    alphaBoolBlend {blendOption=Linear}
    alphaScaleBiasClamp {bInterpResult=true}
    boneToModify {boneName=collar_l__01}
    translation {z=-1}
   }
   n39 = "Transform (Modify) Bone - Bone: collar_r__01"(ComponentPose=@n38, Alpha=@n36, Translation="0,0,1", Rotation="0,0,0", Scale="1,1,1") {
    rotationMode=BMM_Additive
    rotationSpace=BCS_BoneSpace
    translationMode=BMM_Additive
    translationSpace=BCS_BoneSpace
    alphaBoolBlend {blendOption=Linear}
    alphaScaleBiasClamp {bInterpResult=true}
    boneToModify {boneName=collar_r__01}
    translation {z=1}
   }
   n40 = "Transform (Modify) Bone - Bone: deform_upperarm_up_l"(ComponentPose=@n39, Alpha=@n36, Translation="15,0,10", Rotation="0,0,0", Scale="1,1,1") {
    rotationMode=BMM_Additive
    rotationSpace=BCS_BoneSpace
    translationMode=BMM_Additive
    translationSpace=BCS_BoneSpace
    alphaBoolBlend {blendOption=Linear}
    alphaScaleBiasClamp {bInterpResult=true}
    boneToModify {boneName=deform_upperarm_up_l}
    translation {x=15, z=10}
   }
   n41 = "Transform (Modify) Bone - Bone: beltString1_l"(ComponentPose=@n40, Alpha=Get isNotMovingFloat() -> (isNotMovingFloat:float), Translation="0,1,0", Rotation="0,0,0", Scale="1,1,1") {
    rotationMode=BMM_Additive
    rotationSpace=BCS_BoneSpace
    scaleMode=BMM_Additive
    translationMode=BMM_Additive
    translationSpace=BCS_BoneSpace
    alphaBoolBlend {blendOption=Linear}
    alphaScaleBiasClamp {bInterpResult=true}
    boneToModify {boneName=beltString1_l}
    translation {y=1}
   }
   Output Pose(Result=Component To Local(ComponentPose=@n41)) {layerGroup=None}
  ]}
 EventGraph {nodes [
   Event BlueprintUpdateAnimation()
   > n1 = Sequence!(execute) -> (then_0, then_1, then_2:Exec, then_3:Exec)
   > n2 = Gravity Override State Function DEPRECATED(CharacterPartAnimInstance=Self-Reference(), GameTimeInSeconds=GetGameTimeInSeconds(), GravityOverrideParams=Get Gravity Override Params())
   > Set GravityOverride(GravityOverride=@n2)
   > Set isTargetingFloat(isTargetingFloat:float="To Float (Boolean)"(InBool=Get bIsTargeting())) -> (Output_Get:float)
   n3 = Cast To FortPlayerAnimInstance(execute=@n1.then_1, Object=GetAnimInstance(self=GetOwnerSkeletalMeshComponent()))
   > Set isPistol(isPistol=false)
   > Set isRifle(isRifle:bool=false) -> (Output_Get:bool)
   > n4 = Switch on EFortWeaponCoreAnimation!(execute, Selection:EFortWeaponCoreAnimation Enum=Get WeaponCoreAnim(self=@n3)) -> (
    Melee:Exec
    Pistol:Exec
    Shotgun:Exec
    PaperBlueprint:Exec
    Rifle:Exec
    MeleeOneHand:Exec
    MachinePistol:Exec
    RocketLauncher:Exec
    GrenadeLauncher:Exec
    GoingCommando:Exec
    AssaultRifle:Exec
    TacticalShotgun:Exec
    SniperRifle:Exec
    TrapPlacement:Exec
    ShoulderLauncher:Exec
    AbilityDecoTool:Exec
    Crossbow:Exec
    C4:Exec
    RemoteControl:Exec
    DualWield:Exec
    AR_BullPup:Exec
    AR_ForwardGrip:Exec
    MedPackPaddles:Exec
    SMG_P90:Exec
    AR_DrumGun:Exec
    Consumable_Small:Exec
    Consumable_Large:Exec
    Balloon:Exec
    MountedTurret:Exec
    CreativeTool:Exec
    ExplosiveBow:Exec
    AshtonIndigo:Exec
    AshtonChicago:Exec
    MeleeDualWield:Exec
    Unarmed:Exec
   )
   Set isRifle(execute=[@n4.AssaultRifle, @n4.Rifle, @n4.RocketLauncher, @n4.Shotgun, @n4.SniperRifle, @n4.TacticalShotgun], isRifle=true)
   Set isPistol(execute=@n4.Pistol, isPistol=true)
   Set isCrouching(execute=@n1.then_2, isCrouching:float="To Float (Boolean)"(InBool=Get bIsCrouching())) -> (Output_Get:float)
   n5 = NOT Boolean(A=Get bIsTargeting())
   n6 = Get isRifle() -> (isRifle:bool)
   > Set rifleNotTargeting(rifleNotTargeting:float="To Float (Boolean)"(InBool=AND Boolean(A=@n5, B=@n6))) -> (Output_Get:float)
   > Set isSkydiveFloatingFloat(isSkydiveFloatingFloat:float="To Float (Boolean)"(InBool=Get bIsSkydiveFloating())) -> (Output_Get:float)
   n7 = Get bIsSkydiveFloating()
   n8 = Get bIsSkydiveDiveMode()
   > Set skydiveNotGlider(skydiveNotGlider:double="To Float (Boolean)"(InBool=OR Boolean(A=@n8, B=@n7))) -> (Output_Get:double)
   n9 = Get bIsMoving2D()
   > Set isNotMovingFloat(isNotMovingFloat:float="To Float (Boolean)"(InBool=NOT Boolean(A=@n9))) -> (Output_Get:float)
   n10 = Get bIsMoving2D()
   n11 = Get bIsCrouching()
   > Set isCrouchMoving(isCrouchMoving:double="To Float (Boolean)"(InBool=AND Boolean(A=@n11, B=@n10))) -> (Output_Get:double)
   > n12 = Branch(Condition=Get bIsSkyDiving())
   Set IsHangGlider(execute=@n12.else, IsHangGlider:double=0) -> (Output_Get:double)
   n13 = GetAnimInstance(self=GetOwnerSkeletalMeshComponent())
   Is Valid(exec=@n1.then_3, InputObject=@n13)
   > n14 = Set IsCustomEmotePlaying(IsCustomEmotePlaying:double=SelectFloat(A=1, B=0, bPickA=Montage_IsPlaying(self=@n13, Montage=/BRCosmetics/Animation/Game/MainPlayer/Emotes/Mime/Emote_Mime_CMM_M))) -> (Output_Get:double)
   > Set CustomEmoteNotPlaying(CustomEmoteNotPlaying:double="To Float (Boolean)"(InBool="float < float"(A=@n14, B=0.5))) -> (Output_Get:double)
   Event OnParachuteOpened()
   > Set Is Sky Diving(Is Sky Diving:double=0) -> (Output_Get:double)
   > n15 = Switch on EGliderType(Selection=Get PlayerGliderType() -> (PlayerGliderType=HangGlider))
   Set IsHangGlider(execute=[@n15.Cape, @n15.NoGlider, @n15.Surfing, @n15.SurfingSimple, @n15.Umbrella], IsHangGlider=0)
   Delay(execute=@n15.HangGlider, Duration=1)
   > Set IsHangGlider(IsHangGlider=1)
   n16 = Event OnEndSkydiving()
   Set IsHangGlider(execute=[@n16.then, @n17.then], IsHangGlider=0)
   n17 = Event OnParachuteClosed()
  ]}
]
variables [
 CustomEmoteNotPlaying {type=number}
 Gravity Override Params {type=GravityOverrideParamsStruct, default {skydive_FloatingDown {gravityOverride=<0 -700 100>}}}
 GravityOverride {type=Vector}
 Is Sky Diving {type=number}
 isCrouching {type=number}
 isCrouchMoving {type=number}
 IsCustomEmotePlaying {type=number}
 IsHangGlider {type=number}
 isNotMovingFloat {type=number}
 isPistol {type=boolean}
 isRifle {type=boolean}
 isSkydiveFloatingFloat {type=number}
 isTargetingFloat {type=number}
 rifleNotTargeting {type=number}
 skydiveNotGlider {type=number}
]
