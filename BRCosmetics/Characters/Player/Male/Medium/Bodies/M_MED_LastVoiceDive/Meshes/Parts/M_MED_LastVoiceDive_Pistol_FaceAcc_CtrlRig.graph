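# Control Rig for the LastVoiceDive character's pistol face accessory part that shifts the headphones and hat as the right arm moves.
# A spherical pose reader measures upperarm_r relative to spine_05 and outputs a weight for one arm pose region with falloff.
# That weight blends small fixed rotation or translation offsets on headphone_01, headphone_l, headphone_r and hatMain in a chain of Modify Transforms.
# The isCrouching variable is declared but unused in the graph.
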ControlRigBlueprint /BRCosmetics/Characters/Player/Male/Medium/Bodies/M_MED_LastVoiceDive/Meshes/Parts/M_MED_LastVoiceDive_Pistol_FaceAcc_CtrlRig
graphs [
 RigVMFunctionLibrary {links [], nodes []}
 RigVMModel {nodes [
   RigUnit_BeginExecution = &RigUnit_BeginExecution()
   n1 = &RigUnit_SphericalPoseReader(
    ActiveRegionScaleFactors=(PositiveWidth=0.25,NegativeWidth=0.25,PositiveHeight=0.25,NegativeHeight=0.25)
    ActiveRegionSize=0.1
    Debug=(bDrawDebug=false,bDraw2D=false,DebugScale=25,DebugSegments=20,DebugThickness=0.25)
    DriverAxis=(X=-1,Y=0,Z=0)
    DriverItem=(Type=Bone,Name="upperarm_r")
    FalloffRegionScaleFactors=(PositiveWidth=1,NegativeWidth=0.5,PositiveHeight=1,NegativeHeight=1)
    FalloffSize=0.45
    OptionalParentItem=(Type=Bone,Name="spine_05")
    RotationOffset=(X=-100,Y=0,Z=45)
    ExecutePin=@RigUnit_BeginExecution.ExecutePin
   )
   > &RigUnit_ModifyTransforms(
    ItemToModify=((Item=(Type=Bone,Name="headphone_01"),Transform=(Rotation=(X=0,Y=0,Z=0,W=1),Translation=(X=1,Y=0,Z=0),Scale3D=(X=1,Y=1,Z=1))))
    ItemToModify.0.Transform.Rotation=&RigVMFunction_MathQuaternionFromEuler(Euler=(X=-5,Y=0,Z=-10)).Result
    Weight=@n1.OutputParam
   )
   > &RigUnit_ModifyTransforms(
    ItemToModify=((Item=(Type=Bone,Name="headphone_l"),Transform=(Rotation=(X=0,Y=0,Z=0,W=1),Translation=(X=0,Y=0,Z=0),Scale3D=(X=1,Y=1,Z=1))))
    ItemToModify.0.Transform.Rotation=&RigVMFunction_MathQuaternionFromEuler(Euler=(X=0,Y=-10,Z=0)).Result
    Weight=@n1.OutputParam
   )
   > &RigUnit_ModifyTransforms(
    ItemToModify=((Item=(Type=Bone,Name="headphone_r"),Transform=(Rotation=(X=0,Y=0,Z=0,W=1),Translation=(X=0,Y=0,Z=0),Scale3D=(X=1,Y=1,Z=1))))
    ItemToModify.0.Transform.Rotation=&RigVMFunction_MathQuaternionFromEuler(Euler=(X=0,Y=20,Z=0)).Result
    Weight=@n1.OutputParam
   )
   > &RigUnit_ModifyTransforms(
    ItemToModify=((Item=(Type=Bone,Name="hatMain"),Transform=(Rotation=(X=0,Y=0,Z=0,W=1),Translation=(X=-1,Y=0,Z=0),Scale3D=(X=1,Y=1,Z=1))))
    ItemToModify.0.Transform.Rotation=&RigVMFunction_MathQuaternionFromEuler(Euler=(X=0,Y=0,Z=-10)).Result
    Weight=@n1.OutputParam
   )
  ]}
]
hierarchy [
 root {loc=<0 0 0>, rot=<0 0 0 1>}
 pelvis {parent=root, loc=<0 -0.7071 93.1191>, rot=<0.0127 0.707 -0.0127 -0.707>}
 spine_01 {parent=pelvis, loc=<6.7494 -0.6388 0>, rot=<0 0 0.034 -0.9994>}
 spine_02 {parent=spine_01, loc=<7.2518 -0.035 0>, rot=<0 0 0.0238 -0.9997>}
 spine_03 {parent=spine_02, loc=<6.0104 0 0>, rot=<0 0 0.0149 -0.9999>}
 spine_04 {parent=spine_03, loc=<8.1024 0 0>, rot=<0 0 -0.0132 -0.9999>}
 spine_05 {parent=spine_04, loc=<10.0146 0 0>, rot=<0 0 -0.0015 -1>}
 neck_01 {parent=spine_05, loc=<11.3138 0.1529 0>, rot=<0 0 -0.2003 -0.9797>}
 neck_02 {parent=neck_01, loc=<4.5627 0 0>, rot=<0 0 0.0504 -0.9987>}
 head {parent=neck_02, loc=<5.3914 0 0>, rot=<0 0 0.115 -0.9934>}
 hatMain {parent=head, loc=<11.7454 2.1206 0>, rot=<0.0041 -1 0 0>}
 hat_l {parent=hatMain, loc=<0 0 6.6414>, rot=<0 0.3827 0 -0.9239>}
 hat_fr {parent=hatMain, loc=<-0.4138 7.5532 0>, rot=<0.6533 0.2706 -0.2706 -0.6533>}
 hat_bk {parent=hatMain, loc=<0.3446 -7.8841 0>, rot=<-0.6533 0.2706 0.2706 -0.6533>}
 hat_r {parent=hatMain, loc=<-0.0003 0 -6.6414>, rot=<0.9239 0 -0.3827 0>}
 string_MASTER_l {parent=head, loc=<7.8458 5.2123 -6.846>, rot=<0.0041 -1 0 0>}
 string_01_l {parent=string_MASTER_l, loc=<0 0 0>, rot=<0 0 0 1>}
 string_02_l {parent=string_01_l, loc=<4.0348 0 0>, rot=<0 0 0 1>}
 string_03_l {parent=string_02_l, loc=<4.0348 0 0>, rot=<0 0 0 1>}
 string_MASTER_r {parent=head, loc=<7.846 5.2123 6.846>, rot=<0 0 1 -0.0041>}
 string_01_r {parent=string_MASTER_r, loc=<0 0 0>, rot=<0 0 0 1>}
 string_02_r {parent=string_01_r, loc=<4.035 0 0>, rot=<0 0 0 1>}
 string_03_r {parent=string_02_r, loc=<4.035 0 0>, rot=<0 0 0 1>}
 headphone_01 {parent=spine_05, loc=<11.3138 0.1529 0>, rot=<0 0 -0.2003 -0.9797>}
 headphone_l {parent=headphone_01, loc=<6.325 1.4727 -8.9825>, rot=<0.0055 0.0263 -0.7589 -0.6506>}
 headphone_r {parent=headphone_01, loc=<6.325 1.4727 8.9825>, rot=<0.0055 0.0263 -0.7589 -0.6506>}
 clavicle_l {parent=spine_05, loc=<4.3575 -0.1466 -2.3223>, rot=<0.0465 -0.7289 -0.009 -0.6829>}
 upperarm_l {parent=clavicle_l, loc=<15.5111 0 0>, rot=<0.0226 -0.3234 -0.0088 -0.9459>}
 lowerarm_l {parent=upperarm_l, loc=<26.6514 0 0>, rot=<0.0391 -0.0099 -0.2453 -0.9686>}
 clavicle_r {parent=spine_05, loc=<4.3575 -0.1466 2.3223>, rot=<0.7289 0.0465 -0.6829 0.009>}
 upperarm_r {parent=clavicle_r, loc=<-15.5111 0 0>, rot=<0.0226 -0.3234 -0.0088 -0.9459>}
 lowerarm_r {parent=upperarm_r, loc=<-26.6514 0 0>, rot=<0.0391 -0.0099 -0.2453 -0.9686>}
]
variables [isCrouching {type=double, default=0}]
