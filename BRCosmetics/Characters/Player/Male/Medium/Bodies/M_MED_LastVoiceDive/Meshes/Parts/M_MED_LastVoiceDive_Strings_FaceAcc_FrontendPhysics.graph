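# Physics asset for the face-accessory strings on the LastVoiceDive medium male player cosmetic, named as the frontend variant.
# Each side has a three-segment capsule chain with limited swing and zero twist limit, hanging from a zero-size master body.
# Both masters link to the head, and the head to neck_02, with free rotation.
# Head, neck and headphone bodies supply larger capsule and sphere shapes; the headphones have no constraints.
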PhysicsAsset /BRCosmetics/Characters/Player/Male/Medium/Bodies/M_MED_LastVoiceDive/Meshes/Parts/M_MED_LastVoiceDive_Strings_FaceAcc_FrontendPhysics
bodies {
 head [
  {shapeType=Capsule, length=12.5, radius=7.75, shapeName=head_capsule, center=<0.1203 0 0>, rotation=<rot 0 -90.0001 90>}
  {
   shapeType=Capsule
   length=9.625
   radius=5.8438
   shapeName=head_capsule
   center=<2.2347 7.5805 -0.0077>
   rotation=<rot -89.8032 179.9995 -180>
  }
 ]
 headphone_l [
  {shapeType=Sphere, radius=5.2812, shapeName=headphone_l_sphere, center=<10.058 0.5297 1.594>}
  {
   shapeType=Capsule
   length=20.1271
   radius=2.4383
   shapeName=headphone_l_capsule
   center=<1.8436 0.7075 1.3116>
   rotation=<rot 84.0501 -50.5103 -44.3902>
  }
 ]
 headphone_r [
  {shapeType=Sphere, radius=5.2812, shapeName=headphone_l_sphere, center=<10.424 -0.1822 -1.594>}
  {
   shapeType=Capsule
   length=20.1271
   radius=2.4383
   shapeName=headphone_l_capsule
   center=<-1.6952 0.6416 -1.2129>
   rotation=<rot 84.0501 -50.5103 135.6098>
  }
 ]
 neck_02 [{shapeType=Capsule, length=13.25, radius=9.005, shapeName=neck_02_capsule, center=<-3.3426 2.7603 0>, rotation=<rot 0 179.9996 90>}]
 string_01_l [{shapeType=Capsule, length=3.25, radius=0.75, shapeName=string_01_l_capsule, center=<2.6582 0 0>, rotation=<rot 90 0 0>}]
 string_01_r [{shapeType=Capsule, length=3.25, radius=0.75, shapeName=string_01_r_capsule, center=<2.6582 0 0>, rotation=<rot 90 0 0>}]
 string_02_l [{shapeType=Capsule, length=3.25, radius=0.75, shapeName=string_02_l_capsule, center=<2.6582 0 0>, rotation=<rot 90 0 0>}]
 string_02_r [{shapeType=Capsule, length=3.25, radius=0.75, shapeName=string_02_r_capsule, center=<2.6582 0 0>, rotation=<rot 90 0 0>}]
 string_03_l [{shapeType=Capsule, length=3.25, radius=0.75, shapeName=string_03_l_capsule, center=<2.6582 0 0>, rotation=<rot 90 0 0>}]
 string_03_r [{shapeType=Capsule, length=3.25, radius=0.75, shapeName=string_03_r_capsule, center=<2.6582 0 0>, rotation=<rot 90 0 0>}]
 string_MASTER_l [{shapeType=Capsule, length=0, radius=0, shapeName=string_MASTER_l_capsule}]
 string_MASTER_r [{shapeType=Capsule, length=0, radius=0, shapeName=string_MASTER_r_capsule}]
}
constraints [
 {bone1Name=string_01_l, bone2Name=string_MASTER_l, swing1LimitDegrees=20, swing2LimitDegrees=45, twistLimitDegrees=0}
 {bone1Name=string_02_l, bone2Name=string_01_l, swing1LimitDegrees=20, swing2LimitDegrees=45, twistLimitDegrees=0}
 {bone1Name=string_03_l, bone2Name=string_02_l, swing1LimitDegrees=20, swing2LimitDegrees=45, twistLimitDegrees=0}
 {bone1Name=string_01_r, bone2Name=string_MASTER_r, swing1LimitDegrees=20, swing2LimitDegrees=45, twistLimitDegrees=0}
 {bone1Name=string_02_r, bone2Name=string_01_r, swing1LimitDegrees=20, swing2LimitDegrees=45, twistLimitDegrees=0}
 {bone1Name=string_03_r, bone2Name=string_02_r, swing1LimitDegrees=20, swing2LimitDegrees=45, twistLimitDegrees=0}
 {
  bone1Name=string_MASTER_l
  bone2Name=head
  swing1LimitDegrees=45
  swing1Motion=Free
  swing2LimitDegrees=45
  swing2Motion=Free
  twistLimitDegrees=45
  twistMotion=Free
 }
 {
  bone1Name=string_MASTER_r
  bone2Name=head
  swing1LimitDegrees=45
  swing1Motion=Free
  swing2LimitDegrees=45
  swing2Motion=Free
  twistLimitDegrees=45
  twistMotion=Free
 }
 {
  bone1Name=head
  bone2Name=neck_02
  swing1LimitDegrees=45
  swing1Motion=Free
  swing2LimitDegrees=45
  swing2Motion=Free
  twistLimitDegrees=45
  twistMotion=Free
 }
]
