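# Physics asset for the left and right strings on the face accessory of the male medium LastVoiceDive cosmetic body.
# Each side chains three small capsules, with limited swing and zero twist limit, from a zero-size master anchor.
# Both anchors join the head, and the head joins neck_02, with free angular motion.
# Head, neck and both headphones carry larger capsules, with the headphone bodies left unconstrained.
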PhysicsAsset /BRCosmetics/Characters/Player/Male/Medium/Bodies/M_MED_LastVoiceDive/Meshes/Parts/M_MED_LastVoiceDive_Strings_FaceAcc_Physics
bodies {
 head [{shapeType=Capsule, length=12.5, radius=7.75, shapeName=head_capsule, center=<0.1203 0 0>, rotation=<rot 0 -90.0001 90>}]
 headphone_l [{shapeType=Capsule, length=18.9396, radius=3.6258, shapeName=headphone_l_capsule, center=<2.1512 1.2769 0>, rotation=<rot 80 159.9997 180>}]
 headphone_r [{
   shapeType=Capsule
   length=16.9397
   radius=3.5853
   shapeName=headphone_r_capsule
   center=<3.1215 2.1122 -1.9056>
   rotation=<rot 2.0438 -112.9337 -84.4289>
  }]
 neck_02 [{shapeType=Capsule, length=13.25, radius=9.005, shapeName=neck_02_capsule, center=<-3.3426 2.7603 0>, rotation=<rot 0 179.9996 90>}]
 string_01_l [{shapeType=Capsule, length=3.25, radius=0.75, shapeName=string_01_l_capsule, center=<2.6582 0 0>, rotation=<rot 90 0 0>}]
 string_01_r [{shapeType=Capsule, length=3.25, radius=0.75, shapeName=string_01_r_capsule, center=<2.6582 0 0>, rotation=<rot 90 0 0>}]
 string_02_l [{shapeType=Capsule, length=3.25, radius=0.75, shapeName=string_02_l_capsule, center=<2.6582 0 0>, rotation=<rot 90 0 0>}]
 string_02_r [{shapeType=Capsule, length=3.25, radius=0.75, shapeName=string_02_r_capsule, center=<2.6582 0 0>, rotation=<rot 90 0 0>}]
 string_03_l [{shapeType=Capsule, length=3.25, radius=0.75, shapeName=string_03_l_capsule, center=<2.6582 0 0>, rotation=<rot 90 0 0>}]
 string_03_r [{shapeType=Capsule, length=3.25, radius=0.75, shapeName=string_03_r_capsule, center=<2.6582 0 0>, rotation=<rot 90 0 0>}]
 string_MASTER_l [{shapeType=Capsule, length=0, radius=0, shapeName=string_MASTER_l_capsule}]
 string_MASTER_r [{shapeType=Capsule, length=0, radius=0, shapeName=string_MASTER_r_capsule}]
}
constraints [
 {bone1Name=string_01_l, bone2Name=string_MASTER_l, swing1LimitDegrees=20, swing2LimitDegrees=45, twistLimitDegrees=0}
 {bone1Name=string_02_l, bone2Name=string_01_l, swing1LimitDegrees=20, swing2LimitDegrees=45, twistLimitDegrees=0}
 {bone1Name=string_03_l, bone2Name=string_02_l, swing1LimitDegrees=20, swing2LimitDegrees=45, twistLimitDegrees=0}
 {bone1Name=string_01_r, bone2Name=string_MASTER_r, swing1LimitDegrees=20, swing2LimitDegrees=45, twistLimitDegrees=0}
 {bone1Name=string_02_r, bone2Name=string_01_r, swing1LimitDegrees=20, swing2LimitDegrees=45, twistLimitDegrees=0}
 {bone1Name=string_03_r, bone2Name=string_02_r, swing1LimitDegrees=20, swing2LimitDegrees=45, twistLimitDegrees=0}
 {
  bone1Name=string_MASTER_l
  bone2Name=head
  swing1LimitDegrees=45
  swing1Motion=Free
  swing2LimitDegrees=45
  swing2Motion=Free
  twistLimitDegrees=45
  twistMotion=Free
 }
 {
  bone1Name=string_MASTER_r
  bone2Name=head
  swing1LimitDegrees=45
  swing1Motion=Free
  swing2LimitDegrees=45
  swing2Motion=Free
  twistLimitDegrees=45
  twistMotion=Free
 }
 {
  bone1Name=head
  bone2Name=neck_02
  swing1LimitDegrees=45
  swing1Motion=Free
  swing2LimitDegrees=45
  swing2Motion=Free
  twistLimitDegrees=45
  twistMotion=Free
 }
]
