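# Post-process-style Anim Blueprint for the LastVoiceDive outfit's face accessory, animating its hat, headphones and strings on the player.
# Copied body poses pass through Control Rigs toggled by weapon type, crouch and skydive, then headphone offsets and hat AnimDynamics.
# A cached pose, blended with a frontend custom-idle slot, feeds gameplay and frontend RigidBody branches, with skydive bone adjustments on gameplay.
# EventGraph derives gravity override, targeting, stance and weapon flags.
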AnimBlueprint /BRCosmetics/Characters/Player/Male/Medium/Bodies/M_MED_LastVoiceDive/Meshes/Parts/M_MED_LastVoiceDive_FaceAcc_AnimBP
target_skeleton=&/M_MED_LastVoiceDive_FaceAcc_Skeleton
graphs [
 AnimGraph {nodes [
   n1 = RigidBody!(
    ComponentPose=Local To Component(LocalPose=Use cached pose 'PostCustomIdleFilter'())
    Alpha
    OverrideWorldGravity:Vector=Get GravityOverride()
    ExternalForce
   ) -> (Pose) {
    bClampLinearTranslationLimitToRefPose=true
    bOverrideWorldGravity=true
    overridePhysicsAsset=&/M_MED_LastVoiceDive_Strings_FaceAcc_Physics
    simulationSpace=BaseBoneSpace
    alphaBoolBlend {blendOption=Linear}
    baseBoneRef {boneName=head}
    componentAppliedLinearAccClamp=<3000 3000 3000>
    componentLinearAccScale=<0.4 0.3 0.3>
    componentLinearVelScale=<0.4 0.3 0.3>
    overrideWorldGravity {z=-980}
    simSpaceSettings {
     maxAngularAcceleration=1000
     maxAngularVelocity=3
     maxLinearAcceleration=250
     maxLinearVelocity=500
     velocityScaleZ=0
     worldAlpha=0.25
     externalLinearDragV {y=0.5}
    }
   }
   n2 = "Transform (Modify) Bone - Bone: hatMain"!(ComponentPose=@n1, bAlphaBoolEnabled:bool=Get bIsSkydiveFloating(), Translation="0,0,0", Rotation="0,-10,0", Scale="1,1,1") -> (Pose) {
    alphaInputType=Bool
    rotationMode=BMM_Additive
    rotationSpace=BCS_BoneSpace
    translationMode=BMM_Additive
    alphaBoolBlend {blendOption=Linear}
    alphaScaleBiasClamp {bInterpResult=true}
    boneToModify {boneName=hatMain}
    rotation {yaw=-10}
   }
   n3 = "Constraint - headphone_01 (P)"!(ComponentPose=@n2, bAlphaBoolEnabled:bool=Get bIsSkyDiving(), ConstraintWeights_0=1) -> (Pose) {
    alphaInputType=Bool
    alphaBoolBlend {blendOption=Linear}
    boneToModify {boneName=headphone_01}
    constraintSetup [{targetBone {boneName=spine_05}}]
    constraintWeights [1]
   }
   n4 = "Transform (Modify) Bone - Bone: headphone_01"(ComponentPose=@n3, Alpha=Get skydiveNotGlider() -> (skydiveNotGlider:double), Translation="-2,-5,0", Rotation="0,0,0", Scale="1,1,1") {
    rotationMode=BMM_Additive
    rotationSpace=BCS_BoneSpace
    translationMode=BMM_Additive
    translationSpace=BCS_BoneSpace
    alphaBoolBlend {blendOption=Linear}
    alphaScaleBiasClamp {bInterpResult=true, interpSpeedIncreasing=1}
    boneToModify {boneName=headphone_01}
    translation=<-2 -5>
   }
   n5 = RigidBody(ComponentPose=Local To Component(LocalPose=Use cached pose 'PostCustomIdleFilter'()), OverrideWorldGravity=Get GravityOverride()) {
    bClampLinearTranslationLimitToRefPose=true
    bOverrideWorldGravity=true
    overridePhysicsAsset=&/M_MED_LastVoiceDive_Strings_FaceAcc_FrontendPhysics
    simulationSpace=BaseBoneSpace
    alphaBoolBlend {blendOption=Linear}
    baseBoneRef {boneName=head}
    componentAppliedLinearAccClamp=<3000 3000 3000>
    componentLinearAccScale=<0.4 0.3 0.3>
    componentLinearVelScale=<0.4 0.3 0.3>
    overrideWorldGravity {z=-980}
    simSpaceSettings {
     maxAngularAcceleration=1000
     maxAngularVelocity=3
     maxLinearAcceleration=250
     maxLinearVelocity=500
     velocityScaleZ=0
     worldAlpha=0.25
     externalLinearDragV {y=0.5}
    }
   }
   Output Pose(Result=Blend Poses by bool(BlendPose_0=Component To Local(ComponentPose=@n5), BlendPose_1=Component To Local(ComponentPose=@n4)) {blendType=Linear}) {layerGroup=None}
   n6 = Get Body_Mesh()
   n7 = NOT Boolean(A=Get bIsCrouching())
   n8 = Control Rig!(
    Source=Layered blend per bone(BasePose=Copy Pose From Mesh() {bCopyCurves=true, bUseAttachedParent=true}, BlendPoses_0=Copy Pose From Mesh(SourceMeshComponent=@n6)) {layerSetup [{branchFilters [{boneName=upperarm_l}, {boneName=upperarm_r}]}]}
    bAlphaBoolEnabled:bool=@n7
   ) -> (Pose) {
    alphaInputType=Bool
    alphaBoolBlend {blendOption=Linear}
    controlRigAssetReference {blueprintRigClass=&/M_MED_LastVoiceDive_FaceAcc_CtrlRig.M_MED_LastVoiceDive_FaceAcc_CtrlRig_C}
   }
   n9 = Control Rig(Source=@n8, bAlphaBoolEnabled=Get isPistol() -> (isPistol:bool)) {
    alphaInputType=Bool
    alphaBoolBlend {blendOption=Linear}
    controlRigAssetReference {blueprintRigClass=&/M_MED_LastVoiceDive_Pistol_FaceAcc_CtrlRig.M_MED_LastVoiceDive_Pistol_FaceAcc_CtrlRig_C}
   }
   n10 = Control Rig!(
    Source=@n9
    bAlphaBoolEnabled=AND Boolean(A=@n7, B=Get isRifle() -> (isRifle:bool))
    isTargeting:double=Get isTargetingFloat() -> (isTargetingFloat:float)
   ) -> (Pose) {
    alphaInputType=Bool
    alphaBoolBlend {blendOption=Linear}
    controlRigAssetReference {blueprintRigClass=&/M_MED_LastVoiceDive_Rifle_FaceAcc_CtrlRig.M_MED_LastVoiceDive_Rifle_FaceAcc_CtrlRig_C}
   }
   n11 = Control Rig!(Source=@n10, bAlphaBoolEnabled=OR Boolean(A=Get bIsSkyDiving(), B=Get bIsCrouching())) -> (Pose) {
    alphaInputType=Bool
    alphaBoolBlend {blendOption=Linear}
    controlRigAssetReference {blueprintRigClass=&/M_MED_LastVoiceDive_FaceAcc_NeckCrouch_CtrlRig.M_MED_LastVoiceDive_FaceAcc_NeckCrouch_CtrlRig_C}
   }
   n12 = Get isTargetingFloat()
   n13 = "Transform (Modify) Bone - Bone: headphone_r"(ComponentPose=Local To Component(LocalPose=@n11), Alpha=@n12, Translation="0,0,0", Rotation="-10,0,0", Scale="1,1,1") {
    rotationMode=BMM_Additive
    rotationSpace=BCS_BoneSpace
    translationMode=BMM_Additive
    alphaBoolBlend {blendOption=Linear}
    alphaScaleBiasClamp {bInterpResult=true}
    boneToModify {boneName=headphone_r}
    rotation {pitch=-10}
   }
   n14 = "Transform (Modify) Bone - Bone: headphone_l"(ComponentPose=@n13, Alpha=@n12, Translation="0,0,0", Rotation="10,0,0", Scale="1,1,1") {
    rotationMode=BMM_Additive
    rotationSpace=BCS_BoneSpace
    translationMode=BMM_Additive
    alphaBoolBlend {blendOption=Linear}
    alphaScaleBiasClamp {bInterpResult=true}
    boneToModify {boneName=headphone_l}
    rotation {pitch=10}
   }
   n15 = Anim Dynamics - Bone: hat_l(ComponentPose=@n14, Alpha=1, ExternalForce="0,0,0") {
    angularDampingOverride=1
    angularSpringConstant=5000
    bAngularSpring=true
    bOverrideAngularDamping=true
    bUseGravityOverride=true
    gravityScale=0.35
    simulationSpace=RootRelative
    alphaBoolBlend {blendOption=Linear}
    boundBone {boneName=hat_l}
    componentLinearAccScale=<1 1 1>
    componentLinearVelScale=<1 1 1>
    gravityOverride {z=-1500}
    physicsBodyDefinitions [{
      boundBone {boneName=hat_l}
      boxExtents=<15 15 15>
      constraintSetup {angularLimitsMax=<30 30 30>, angularLimitsMin=<-30 -30 -30>, angularTarget=<1 0 0>}
      localJointOffset=<1 0 0>
     }]
   }
   n16 = Anim Dynamics - Bone: hat_fr(ComponentPose=@n15, Alpha=1, ExternalForce="0,0,0") {
    angularDampingOverride=1
    angularSpringConstant=5000
    bAngularSpring=true
    bOverrideAngularDamping=true
    bUseGravityOverride=true
    gravityScale=0.35
    simulationSpace=RootRelative
    alphaBoolBlend {blendOption=Linear}
    boundBone {boneName=hat_fr}
    componentLinearAccScale=<1 1 1>
    componentLinearVelScale=<1 1 1>
    gravityOverride {z=-1500}
    physicsBodyDefinitions [{
      boundBone {boneName=hat_fr}
      boxExtents=<15 15 15>
      constraintSetup {angularLimitsMax=<30 30 30>, angularLimitsMin=<-30 -30 -30>, angularTarget=<1 0 0>}
      localJointOffset=<1 0 0>
     }]
   }
   n17 = Anim Dynamics - Bone: hat_bk(ComponentPose=@n16, Alpha=1, ExternalForce="0,0,0") {
    angularDampingOverride=1
    angularSpringConstant=5000
    bAngularSpring=true
    bOverrideAngularDamping=true
    bUseGravityOverride=true
    gravityScale=0.35
    simulationSpace=RootRelative
    alphaBoolBlend {blendOption=Linear}
    boundBone {boneName=hat_bk}
    componentLinearAccScale=<1 1 1>
    componentLinearVelScale=<1 1 1>
    gravityOverride {z=-1500}
    physicsBodyDefinitions [{
      boundBone {boneName=hat_bk}
      boxExtents=<15 15 15>
      constraintSetup {angularLimitsMax=<30 0 30>, angularLimitsMin=<-30 -30 -30>, angularTarget=<1 0 0>}
      localJointOffset=<1 0 0>
     }]
   }
   n18 = Anim Dynamics - Bone: hat_r(ComponentPose=@n17, Alpha=1, ExternalForce="0,0,0") {
    angularDampingOverride=1
    angularSpringConstant=5000
    bAngularSpring=true
    bOverrideAngularDamping=true
    bUseGravityOverride=true
    gravityScale=0.35
    simulationSpace=RootRelative
    alphaBoolBlend {blendOption=Linear}
    boundBone {boneName=hat_r}
    componentLinearAccScale=<1 1 1>
    componentLinearVelScale=<1 1 1>
    gravityOverride {z=-1500}
    physicsBodyDefinitions [{
      boundBone {boneName=hat_r}
      boxExtents=<15 15 15>
      constraintSetup {angularLimitsMax=<30 30 30>, angularLimitsMin=<-30 -30 -30>, angularTarget=<1 0 0>}
      localJointOffset=<1 0 0>
     }]
   }
   Save cached pose 'PostDynamics'(Pose=Component To Local(ComponentPose=@n18)) {cacheName=PostDynamics}
   Save cached pose 'PostCustomIdleFilter'(Pose=Layered blend per bone(
     BasePose=Use cached pose 'PostDynamics'()
     BlendPoses_0=Slot 'DefaultSlot'()
     BlendWeights_0=Get Custom Idle is Playing() -> (Custom Idle is Playing:double)
    ) {layerSetup [{branchFilters [{boneName=hatMain}]}]}) {cacheName=PostCustomIdleFilter}
  ]}
 EventGraph {nodes [
   Event BlueprintUpdateAnimation()
   > n1 = Sequence!(execute) -> (then_0, then_1, then_2:Exec, then_3:Exec, then_4:Exec)
   > n2 = Gravity Override State Function DEPRECATED(CharacterPartAnimInstance=Self-Reference(), GameTimeInSeconds=GetGameTimeInSeconds(), GravityOverrideParams=Get Gravity Override Params())
   > Set GravityOverride(GravityOverride=@n2)
   > Set isTargetingFloat(isTargetingFloat:float="To Float (Boolean)"(InBool=Get bIsTargeting())) -> (Output_Get:float)
   n3 = Set isCrouching(execute=@n1.then_2, isCrouching:float="To Float (Boolean)"(InBool=Get bIsCrouching())) -> (Output_Get:float)
   n4 = Get isPistol()
   > Set isPistolNoCrouch(isPistolNoCrouch:double=float * float(A=float - float(A=1, B=@n3), B="To Float (Boolean)"(InBool=@n4)) {is=title K2Node_PromotableOperator}) -> (Output_Get:double)
   n5 = Get bIsSkydiveDiveMode()
   n6 = Get bIsSkydiveFloating()
   Set skydiveNotGlider(execute=@n1.then_3, skydiveNotGlider:double="To Float (Boolean)"(InBool=OR Boolean(A=@n5, B=@n6))) -> (Output_Get:double)
   n7 = Cast To FortPlayerAnimInstance(execute=@n1.then_1, Object=GetAnimInstance(self=GetOwnerSkeletalMeshComponent()))
   > Set isPistol(isPistol=false)
   > Set isRifle(isRifle:bool=false) -> (Output_Get:bool)
   > n8 = Switch on EFortWeaponCoreAnimation!(execute, Selection:EFortWeaponCoreAnimation Enum=Get WeaponCoreAnim(self=@n7)) -> (
    Melee:Exec
    Pistol:Exec
    Shotgun:Exec
    PaperBlueprint:Exec
    Rifle:Exec
    MeleeOneHand:Exec
    MachinePistol:Exec
    RocketLauncher:Exec
    GrenadeLauncher:Exec
    GoingCommando:Exec
    AssaultRifle:Exec
    TacticalShotgun:Exec
    SniperRifle:Exec
    TrapPlacement:Exec
    ShoulderLauncher:Exec
    AbilityDecoTool:Exec
    Crossbow:Exec
    C4:Exec
    RemoteControl:Exec
    DualWield:Exec
    AR_BullPup:Exec
    AR_ForwardGrip:Exec
    MedPackPaddles:Exec
    SMG_P90:Exec
    AR_DrumGun:Exec
    Consumable_Small:Exec
    Consumable_Large:Exec
    Balloon:Exec
    MountedTurret:Exec
    CreativeTool:Exec
    ExplosiveBow:Exec
    AshtonIndigo:Exec
    AshtonChicago:Exec
    MeleeDualWield:Exec
    Unarmed:Exec
   )
   Set isRifle(execute=[@n8.AssaultRifle, @n8.Rifle, @n8.RocketLauncher, @n8.Shotgun, @n8.SniperRifle, @n8.TacticalShotgun], isRifle=true)
   Set isPistol(execute=@n8.Pistol, isPistol=true)
   Branch(execute=@n1.then_4, Condition=Get bIsInFrontEnd())
   > Set Custom Idle is Playing(Custom Idle is Playing:double=GetCurveValue(self=GetOwnerFrontEndAnimInstance(), CurveName=Frontend_Idle_Bypass)) -> (Output_Get:double)
   Event BlueprintInitializeAnimation()
   > Set Body_Mesh(Body_Mesh=GetOwnerSkeletalMeshComponent())
  ]}
]
variables [
 body Mesh {type=SkeletalMeshComponent}
 Body_Mesh {type=SkeletalMeshComponent}
 Custom Idle is Playing {type=number}
 Gravity Override Params {type=GravityOverrideParamsStruct, default {skydive_FloatingDown {gravityOverride=<0 -700 100>}}}
 GravityOverride {type=Vector}
 isCrouching {type=number}
 isPistol {type=boolean}
 isPistolNoCrouch {type=number}
 isRifle {type=boolean}
 isTargetingFloat {type=number}
 skydiveNotGlider {type=number}
]
